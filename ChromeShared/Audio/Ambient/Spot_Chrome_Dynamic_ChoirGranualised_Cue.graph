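# Sound cue for spot one-shot ambience in the Chrome biome, named as granulised choir and built from Chrome wall ambience waves.
# A random node picks one of sixteen equally weighted waves and feeds a modulator that varies its volume.
# The quality switch shares that branch across its first three slots and leaves the fourth empty.
# Custom attenuation applies, virtualization is disabled, and a concurrency override caps simultaneous instances, rejecting new ones.
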SoundCue /ChromeShared/Audio/Ambient/Spot_Chrome_Dynamic_ChoirGranualised_Cue
root=QualityLevel_0
nodes [
 {id=Modulator_0, children [Random_2], properties {volumeMax=1, volumeMin=0.5}}
 {id=QualityLevel_0, children [Modulator_0,Modulator_0,Modulator_0,null]}
 {
  id=Random_2
  children [
   WavePlayer_16
   WavePlayer_17
   WavePlayer_18
   WavePlayer_19
   WavePlayer_20
   WavePlayer_21
   WavePlayer_22
   WavePlayer_23
   WavePlayer_24
   WavePlayer_25
   WavePlayer_26
   WavePlayer_27
   WavePlayer_28
   WavePlayer_29
   WavePlayer_30
   WavePlayer_31
  ]
  properties {bShouldExcludeFromBranchCulling=true, weights [1,1,1,1,1,1,1,1,1,1,1,1,1,1,1,1]}
 }
 {id=WavePlayer_16, properties {soundWaveAssetPtr=/Game/Sounds/Artemis/Biomes/Chrome/Spots/Spot_Chrome_Wall_Amb_01}}
 {id=WavePlayer_17, properties {soundWaveAssetPtr=/Game/Sounds/Artemis/Biomes/Chrome/Spots/Spot_Chrome_Wall_Amb_02}}
 {id=WavePlayer_18, properties {soundWaveAssetPtr=/Game/Sounds/Artemis/Biomes/Chrome/Spots/Spot_Chrome_Wall_Amb_03}}
 {id=WavePlayer_19, properties {soundWaveAssetPtr=/Game/Sounds/Artemis/Biomes/Chrome/Spots/Spot_Chrome_Wall_Amb_04}}
 {id=WavePlayer_20, properties {soundWaveAssetPtr=/Game/Sounds/Artemis/Biomes/Chrome/Spots/Spot_Chrome_Wall_Amb_05}}
 {id=WavePlayer_21, properties {soundWaveAssetPtr=/Game/Sounds/Artemis/Biomes/Chrome/Spots/Spot_Chrome_Wall_Amb_06}}
 {id=WavePlayer_22, properties {soundWaveAssetPtr=/Game/Sounds/Artemis/Biomes/Chrome/Spots/Spot_Chrome_Wall_Amb_07}}
 {id=WavePlayer_23, properties {soundWaveAssetPtr=/Game/Sounds/Artemis/Biomes/Chrome/Spots/Spot_Chrome_Wall_Amb_08}}
 {id=WavePlayer_24, properties {soundWaveAssetPtr=/Game/Sounds/Artemis/Biomes/Chrome/Spots/Spot_Chrome_Wall_Amb_09}}
 {id=WavePlayer_25, properties {soundWaveAssetPtr=/Game/Sounds/Artemis/Biomes/Chrome/Spots/Spot_Chrome_Wall_Amb_10}}
 {id=WavePlayer_26, properties {soundWaveAssetPtr=/Game/Sounds/Artemis/Biomes/Chrome/Spots/Spot_Chrome_Wall_Amb_11}}
 {id=WavePlayer_27, properties {soundWaveAssetPtr=/Game/Sounds/Artemis/Biomes/Chrome/Spots/Spot_Chrome_Wall_Amb_12}}
 {id=WavePlayer_28, properties {soundWaveAssetPtr=/Game/Sounds/Artemis/Biomes/Chrome/Spots/Spot_Chrome_Wall_Amb_13}}
 {id=WavePlayer_29, properties {soundWaveAssetPtr=/Game/Sounds/Artemis/Biomes/Chrome/Spots/Spot_Chrome_Wall_Amb_14}}
 {id=WavePlayer_30, properties {soundWaveAssetPtr=/Game/Sounds/Artemis/Biomes/Chrome/Spots/Spot_Chrome_Wall_Amb_15}}
 {id=WavePlayer_31, properties {soundWaveAssetPtr=/Game/Sounds/Artemis/Biomes/Chrome/Spots/Spot_Chrome_Wall_Amb_16}}
]
properties {
 attenuationSettings=&/Chrome_DynamicAmbient_Att
 bEnableBusSends=false
 bOverrideConcurrency=true
 maxDistance=6000
 soundClassObject=&Fort_Ambient_OneShots
 virtualizationMode=Disabled
 volumeMultiplier=2.005
 concurrencyOverrides {maxCount=3, resolutionRule=PreventNew}
 preEffectBusSends [{}]
}
