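# Sound Cue for a spot ambient one-shot in the Chrome biome, playing granulated horror whooshes.
# A Random node picks among four equally weighted wave variants and feeds a default-setting Modulator for pitch and volume variation.
# The root Quality Level node routes that chain to its first three inputs and leaves the fourth empty.
# It uses Chrome dynamic ambient attenuation, disables virtualization, and overrides concurrency to block new plays beyond three instances.
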SoundCue /ChromeShared/Audio/Ambient/Spot_Chrome_Dynamic_HorrorWhooshGranualized_Cue
root=QualityLevel_0
nodes [
 {id=Modulator_0, children [Random_2]}
 {id=QualityLevel_0, children [Modulator_0,Modulator_0,Modulator_0,null]}
 {
  id=Random_2
  children [WavePlayer_39, WavePlayer_40, WavePlayer_41, WavePlayer_42]
  properties {bShouldExcludeFromBranchCulling=true, weights [1,1,1,1]}
 }
 {id=WavePlayer_39, properties {soundWaveAssetPtr=/Game/Sounds/Artemis/Biomes/Chrome/Spots/Spot_Chrome_Horror_Whoosh_Granualised_04}}
 {id=WavePlayer_40, properties {soundWaveAssetPtr=/Game/Sounds/Artemis/Biomes/Chrome/Spots/Spot_Chrome_Horror_Whoosh_Granualised_03}}
 {id=WavePlayer_41, properties {soundWaveAssetPtr=/Game/Sounds/Artemis/Biomes/Chrome/Spots/Spot_Chrome_Horror_Whoosh_Granualised_02}}
 {id=WavePlayer_42, properties {soundWaveAssetPtr=/Game/Sounds/Artemis/Biomes/Chrome/Spots/Spot_Chrome_Horror_Whoosh_Granualised_01}}
]
properties {
 attenuationSettings=&/Chrome_DynamicAmbient_Att
 bEnableBusSends=false
 bOverrideConcurrency=true
 maxDistance=6000
 soundClassObject=&Fort_Ambient_OneShots
 virtualizationMode=Disabled
 volumeMultiplier=0.752
 concurrencyOverrides {maxCount=3, resolutionRule=PreventNew}
 preEffectBusSends [{}]
}
